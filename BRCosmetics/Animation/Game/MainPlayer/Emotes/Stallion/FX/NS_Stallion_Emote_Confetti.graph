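# Niagara confetti effect for the Stallion emote, drifting red and off-white flecks down from overhead.
# A GPU Fountain emitter and a lower-rate FountainMobile variant pulse spawning via two multiplied age curves, inside a cylinder whose offset shifts over time.
# Sprites get array-picked colors and SubUV frames, then tumble under gravity, waveform-modulated wind, aerodynamic drag and collision, aligned to mesh orientation.
# Fountain001 and SingleLoopingParticle are disabled.
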NiagaraSystem /BRCosmetics/Animation/Game/MainPlayer/Emotes/Stallion/FX/NS_Stallion_Emote_Confetti
dependencies {
 usedDataInterfaces [&NiagaraDataInterfaceCurve, &NiagaraDataInterfaceArrayColor]
 usedDynamicInputs []
 usedModules [&SystemState]
}
emitters {
 Fountain {
  simTarget=GPUComputeSim
  emitterUpdateScript [
   &EmitterState(Life Cycle Mode=::NewEnumerator1, Loop Duration=2, UseLoopDelay=false)
   &SpawnRate(
    SpawnRate=&Multiply_Float(
     A=&FloatFromCurve(
      FloatCurve=DataInterface() {propertyValues {Curve {keys [
          {interpMode=RCIM_Cubic, tangentMode=RCTM_User, time=0, value=0}
          {arriveTangent=2, interpMode=RCIM_Cubic, leaveTangent=2, time=0.3, value=1}
          {arriveTangent=-6.7237, interpMode=RCIM_Cubic, leaveTangent=-6.7237, tangentMode=RCTM_User, time=0.5, value=1}
          {tangentMode=RCTM_User, time=1, value=0}
         ]}}}
      CurveIndex=$Emitter.NormalizedLoopAge
      Scale Curve=250
     )
     B=&FloatFromCurve(
      FloatCurve=DataInterface() {propertyValues {Curve {keys [
          {arriveTangent=0.1751, interpMode=RCIM_Cubic, leaveTangent=0.1751, tangentMode=RCTM_User, time=0, value=0.1}
          <at 0.8 0.2>
          {arriveTangent=1.4308, interpMode=RCIM_Cubic, leaveTangent=1.4308, tangentMode=RCTM_User, time=1.3, value=5}
          <at 2.5 0.4>
          {arriveTangent=-48.9284, interpMode=RCIM_Cubic, leaveTangent=-48.9284, tangentMode=RCTM_User, time=3.5, value=0.5}
         ]}}}
      CurveIndex=$Emitter.Age
     )
    )
    Use Spawn Probability=true
   )
  ]
  particleSpawnScript [
   &InitializeParticle(
    Color=&HueShiftLinearColor(
     Hue=0
     Saturation=1
     Value=&RandomRangeFloat(Minimum=0, Maximum=1)
     Linear Color=&SelectLinearColorFromArray(
      Array Sampling Mode=::NewEnumerator1
      Direct Array Index=&ReturnExecIndex()
      Direct Array Mode=::NewEnumerator1
      Color Selection Array=DataInterface() {propertyValues {ColorData [<rgba 0.8 0.8 0.8 1>, <rgba 0.5 0.0306 0.0306 1>, <rgba 0.2747 0.0168 0.0168 1>, <rgba 0.5 0.0306 0.0306 1>, <rgba 0.1 0.0061 0.0061 1>]}}
      Random Seed=?
     )
    )
    Lifetime Max=5
    Lifetime Min=2
    Lifetime Mode=::NewEnumerator1
    Lifetime Randomness Mode:ENiagaraRandomnessMode=::NewEnumerator1
    Mass Mode=::NewEnumerator3
    Sprite Rotation Mode=::NewEnumerator2
    Sprite Size Max=<3 5>
    Sprite Size Min=<2 2>
    Sprite Size Mode=::NewEnumerator2
    UsePositionOffset=false
    Write Material Random=true
   )
   &ShapeLocation(
    Cylinder Height=120
    Cylinder Radius=150
    MaskSpawnGroup=false
    Offset=&Lerp_Vector(
     A=<0 0 200>
     B=<1 100 180>
     Alpha=&FloatFromCurve(
      FloatCurve=DataInterface() {propertyValues {Curve {keys [
          {arriveTangent=0.1751, interpMode=RCIM_Cubic, leaveTangent=0.1751, tangentMode=RCTM_User, time=0, value=0}
          <at 1.2 0.2>
          {arriveTangent=1.4308, interpMode=RCIM_Cubic, leaveTangent=1.4308, tangentMode=RCTM_User, time=1.3, value=1}
          {arriveTangent=-0.1304, interpMode=RCIM_Cubic, leaveTangent=-0.1304, tangentMode=RCTM_User, time=2.7, value=0}
         ]}}}
      CurveIndex=$Emitter.Age
     )
    )
    Offset Coordinate Space=::Local
    Offset Mode=::NewEnumerator0
    Override Seed=false
    Shape Primitive=::NewEnumerator1
    Surface Only=false
    UseOwnerScale=true
   )
   &AddVelocity(
    Cone Angle=30
    Cone Axis=<0 0 1>
    Distribution Along Cone Axis=0.25
    Module.UseConeFalloff=true
    Module.UseVelDistribution=false
    Override Seed=false
    Speed Falloff From Cone Axis=1
    Velocity Mode=::NewEnumerator2
    Velocity Speed=&RandomRangeFloat(Minimum=100, Maximum=50)
   ) {enabled=false}
   &InitialMeshOrientation(Use Rotation Vector=false)
   &SubUVAnimation(End Frame=35, SubUV Animation Mode:ENiagara_SubUVLookupMode=::NewEnumerator2)
   &InheritVelocity(Speed Limit=true, UseSpeedThreshold=false)
   &ApplyInitialForces()
  ]
  particleUpdateScript [
   &ParticleState()
   &ScaleSpriteSize(Uniform Curve Sprite Scale=DataInterface() {propertyValues {Curve {keys [
        {arriveTangent=32.4705, interpMode=RCIM_Cubic, leaveTangent=32.4705, tangentMode=RCTM_User, time=0, value=0}
        {arriveTangent=0.9995, interpMode=RCIM_Cubic, leaveTangent=0.9995, tangentMode=RCTM_User, time=0.1, value=1}
        <at 0.9 1>
        {arriveTangent=-0.0405, interpMode=RCIM_Cubic, leaveTangent=-0.0405, tangentMode=RCTM_User, time=1, value=0}
       ]}}})
   &GravityForce()
   &WindForce(
    Air Resistance=0.75
    Wind Speed=<2000 0 500>
    Wind Speed Scale=&Lerp_Float(
     A=0.25
     B=3
     Alpha=&Waveform("Waveform [1]"=::NewEnumerator9, "[1] Frequency"=0.5, "[1] Amplitude Min/Max"=<0 1>)
    )
   )
   &AerodynamicDrag(
    Aerodynamic Rotational Drag=0.2
    Evaluation Type:ENiagaraRandomnessEvaluation=::NewEnumerator1
    Rotation Strength=15
    Use Linear Drag=true
    Use Rotational Drag=true
   )
   &Collision(
    Age Colliding Particles=false
    Control Roll On Collision=true
    Kill Offscreen Particles=true
    Module.Control Rest State Via Surface Normal=false
    Module.EnableMaxCollisionCount=false
    Module.Randomize Collision Normal Vector=false
    Restitution=0.1
    ScaleRateOfRotation=false
   )
   &SolveForcesAndVelocity(Clamp Velocity=false, Limit Acceleration=false)
   &AlignSpriteToMeshOrientation(SetAlignment=true, SetFacing=true)
  ]
  rendererClasses [&NiagaraSpriteRendererProperties]
  renderers [{rendererClass=&NiagaraSpriteRendererProperties}]
 }
 Fountain001 {
  bEnabled=false
  emitterUpdateScript [
   &EmitterState(Life Cycle Mode=::NewEnumerator1, Loop Duration=2, UseLoopDelay=false)
   &SpawnRate(SpawnRate=90, Use Spawn Probability=false)
  ]
  particleSpawnScript [
   &InitializeParticle(
    Lifetime Max=1.75
    Lifetime Min=1.4
    Lifetime Mode=::NewEnumerator1
    Mass Mode=::NewEnumerator3
    Sprite Rotation Mode=::NewEnumerator2
    Sprite Size Mode=::NewEnumerator1
    Uniform Sprite Size Max=12
    Uniform Sprite Size Min=6
    UsePositionOffset=false
    Write Material Random=true
   )
   &ShapeLocation(MaskSpawnGroup=false, Override Seed=false, Sphere Radius=8, UseOwnerScale=true)
   &AddVelocity(
    Cone Angle=32
    Cone Axis=<0 0 1>
    Distribution Along Cone Axis=0.25
    Module.UseConeFalloff=true
    Module.UseVelDistribution=true
    Override Seed=false
    Speed Falloff From Cone Axis=0.5
    Velocity Mode=::NewEnumerator2
    Velocity Speed=&RandomRangeFloat(Minimum=500, Maximum=850)
   )
  ]
  particleUpdateScript [
   &ParticleState()
   &GravityForce()
   &Drag(Drag=0.25, Use Linear Drag=true, Use Rotational Drag=false)
   &ScaleColor(
    Scale Alpha=&FloatFromCurve(FloatCurve=DataInterface() {propertyValues {Curve {keys [<at -0.0046 1.0113 RCIM_Cubic>,<at 1.0011 0.002 RCIM_Cubic>]}}})
    ScaleA=true
    ScaleRGB=true
   )
   &SolveForcesAndVelocity(Clamp Velocity=false, Limit Acceleration=false)
  ]
  rendererClasses [&NiagaraSpriteRendererProperties]
  renderers [{rendererClass=&NiagaraSpriteRendererProperties}]
 }
 FountainMobile {
  emitterUpdateScript [
   &EmitterState(Life Cycle Mode=::NewEnumerator1, Loop Duration=2, UseLoopDelay=false)
   &SpawnRate(
    SpawnRate=&Multiply_Float(
     A=&FloatFromCurve(
      FloatCurve=DataInterface() {propertyValues {Curve {keys [
          {interpMode=RCIM_Cubic, tangentMode=RCTM_User, time=0, value=0}
          {arriveTangent=2, interpMode=RCIM_Cubic, leaveTangent=2, time=0.3, value=1}
          {arriveTangent=-6.7237, interpMode=RCIM_Cubic, leaveTangent=-6.7237, tangentMode=RCTM_User, time=0.5, value=1}
          {tangentMode=RCTM_User, time=1, value=0}
         ]}}}
      CurveIndex=$Emitter.NormalizedLoopAge
      Scale Curve=50
     )
     B=&FloatFromCurve(
      FloatCurve=DataInterface() {propertyValues {Curve {keys [
          {arriveTangent=0.1751, interpMode=RCIM_Cubic, leaveTangent=0.1751, tangentMode=RCTM_User, time=0, value=0.1}
          <at 0.8 0.2>
          {arriveTangent=1.4308, interpMode=RCIM_Cubic, leaveTangent=1.4308, tangentMode=RCTM_User, time=1.3, value=5}
          <at 2.5 0.4>
          {arriveTangent=-48.9284, interpMode=RCIM_Cubic, leaveTangent=-48.9284, tangentMode=RCTM_User, time=3.5, value=0.5}
         ]}}}
      CurveIndex=$Emitter.Age
     )
    )
    Use Spawn Probability=true
   )
  ]
  particleSpawnScript [
   &InitializeParticle(
    Color=&HueShiftLinearColor(
     Hue=0
     Saturation=1
     Value=&RandomRangeFloat(Minimum=0, Maximum=1)
     Linear Color=&SelectLinearColorFromArray(
      Array Sampling Mode=::NewEnumerator1
      Direct Array Index=&ReturnExecIndex()
      Direct Array Mode=::NewEnumerator1
      Color Selection Array=DataInterface() {propertyValues {ColorData [<rgba 0.8 0.8 0.8 1>, <rgba 0.5 0.0306 0.0306 1>, <rgba 0.2747 0.0168 0.0168 1>, <rgba 0.5 0.0306 0.0306 1>, <rgba 0.1 0.0061 0.0061 1>]}}
      Random Seed=?
     )
    )
    Lifetime Max=5
    Lifetime Min=2
    Lifetime Mode=::NewEnumerator1
    Lifetime Randomness Mode:ENiagaraRandomnessMode=::NewEnumerator1
    Mass Mode=::NewEnumerator3
    Sprite Rotation Mode=::NewEnumerator2
    Sprite Size Max=<3 5>
    Sprite Size Min=<2 2>
    Sprite Size Mode=::NewEnumerator2
    UsePositionOffset=false
    Write Material Random=true
   )
   &ShapeLocation(
    Cylinder Height=120
    Cylinder Radius=150
    MaskSpawnGroup=false
    Offset=&Lerp_Vector(
     A=<0 0 200>
     B=<1 100 180>
     Alpha=&FloatFromCurve(
      FloatCurve=DataInterface() {propertyValues {Curve {keys [
          {arriveTangent=0.1751, interpMode=RCIM_Cubic, leaveTangent=0.1751, tangentMode=RCTM_User, time=0, value=0}
          <at 1.2 0.2>
          {arriveTangent=1.4308, interpMode=RCIM_Cubic, leaveTangent=1.4308, tangentMode=RCTM_User, time=1.3, value=1}
          {arriveTangent=-0.1304, interpMode=RCIM_Cubic, leaveTangent=-0.1304, tangentMode=RCTM_User, time=2.7, value=0}
         ]}}}
      CurveIndex=$Emitter.Age
     )
    )
    Offset Coordinate Space=::Local
    Offset Mode=::NewEnumerator0
    Override Seed=false
    Shape Primitive=::NewEnumerator1
    Surface Only=false
    UseOwnerScale=true
   )
   &AddVelocity(
    Cone Angle=30
    Cone Axis=<0 0 1>
    Distribution Along Cone Axis=0.25
    Module.UseConeFalloff=true
    Module.UseVelDistribution=false
    Override Seed=false
    Speed Falloff From Cone Axis=1
    Velocity Mode=::NewEnumerator2
    Velocity Speed=&RandomRangeFloat(Minimum=100, Maximum=50)
   ) {enabled=false}
   &InitialMeshOrientation(Use Rotation Vector=false)
   &SubUVAnimation(End Frame=35, SubUV Animation Mode:ENiagara_SubUVLookupMode=::NewEnumerator2)
   &InheritVelocity(Speed Limit=true, UseSpeedThreshold=false)
   &ApplyInitialForces()
  ]
  particleUpdateScript [
   &ParticleState()
   &ScaleSpriteSize(Uniform Curve Sprite Scale=DataInterface() {propertyValues {Curve {keys [
        {arriveTangent=32.4705, interpMode=RCIM_Cubic, leaveTangent=32.4705, tangentMode=RCTM_User, time=0, value=0}
        {arriveTangent=0.9995, interpMode=RCIM_Cubic, leaveTangent=0.9995, tangentMode=RCTM_User, time=0.1, value=1}
        <at 0.9 1>
        {arriveTangent=-0.0405, interpMode=RCIM_Cubic, leaveTangent=-0.0405, tangentMode=RCTM_User, time=1, value=0}
       ]}}})
   &GravityForce()
   &WindForce(
    Air Resistance=0.75
    Wind Speed=<2000 0 500>
    Wind Speed Scale=&Lerp_Float(
     A=0.25
     B=3
     Alpha=&Waveform("Waveform [1]"=::NewEnumerator9, "[1] Frequency"=0.5, "[1] Amplitude Min/Max"=<0 1>)
    )
   )
   &AerodynamicDrag(
    Aerodynamic Rotational Drag=0.2
    Evaluation Type:ENiagaraRandomnessEvaluation=::NewEnumerator1
    Rotation Strength=15
    Use Linear Drag=true
    Use Rotational Drag=true
   )
   &Collision(
    Age Colliding Particles=false
    Control Roll On Collision=true
    Module.Control Rest State Via Surface Normal=false
    Module.EnableMaxCollisionCount=false
    Module.Randomize Collision Normal Vector=false
    Restitution=0.1
    ScaleRateOfRotation=false
   )
   &SolveForcesAndVelocity(Clamp Velocity=false, Limit Acceleration=false)
   &AlignSpriteToMeshOrientation(SetAlignment=true, SetFacing=true)
  ]
  rendererClasses [&NiagaraSpriteRendererProperties]
  renderers [{rendererClass=&NiagaraSpriteRendererProperties}]
 }
 SingleLoopingParticle {
  bEnabled=false
  emitterUpdateScript [
   &EmitterState()
   &SpawnBurst_Instantaneous(Use Spawn Probability=false, UseLoopCountLimit=false)
  ]
  particleSpawnScript [&InitializeParticle(
    Color=<rgba 0 0.5 1 1>
    Lifetime=$Emitter.CurrentLoopDuration
    Sprite Size Mode=::NewEnumerator3
    UsePositionOffset=false
    Write Material Random=true
   )]
  particleUpdateScript [
   &ParticleState()
   &SolveForcesAndVelocity(Clamp Velocity=false, Limit Acceleration=false)
  ]
  rendererClasses [&NiagaraSpriteRendererProperties]
  renderers [{rendererClass=&NiagaraSpriteRendererProperties}]
 }
}
summary {
 systemName=NS_Stallion_Emote_Confetti
 emitters [
  {emitterName=Fountain, simTarget=GPUComputeSim, rendererClasses [&NiagaraSpriteRendererProperties]}
  {bEnabled=false, emitterName=SingleLoopingParticle, rendererClasses [&NiagaraSpriteRendererProperties]}
  {bEnabled=false, emitterName=Fountain001, rendererClasses [&NiagaraSpriteRendererProperties]}
  {emitterName=FountainMobile, rendererClasses [&NiagaraSpriteRendererProperties]}
 ]
 userVariables []
}
